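# Animation Blueprint for the TowerMilk backpack cosmetic that plays its idle animation.
# The output pose comes straight from a sequence player running Backpack_TowerMilk_Idle.
# An unconnected "Idle + IdleBreak" state machine holds a random player weighted toward idle over idle break, so the break never reaches the output.
# The Play_IdleBreak variable is unused, and the event graph holds only an unwired pawn-owner lookup.
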AnimBlueprint /BRCosmetics/Accessories/FORT_Backpacks/Backpack_M_MED_TowerMilk/Meshes/M_MED_TowerMilk_Pack_AnimBP
target_skeleton=&/M_MED_TowerMilk_Pack_Skeleton
graphs [
 AnimGraph {nodes [
   Input Pose()
   Output Pose(Result=Sequence Player 'Backpack_TowerMilk_Idle'() {sequence=/BRCosmetics/Accessories/FORT_Backpacks/Backpack_M_MED_TowerMilk/Animations/Backpack_TowerMilk_Idle}) {layerGroup=None}
   AnimGraphNode_StateMachine_0 = Idle + IdleBreak() {is=title AnimGraphNode_StateMachine}
  ]}
 Idle + IdleBreak {
  entry=Idle Anims
  path=AnimGraph.AnimGraphNode_StateMachine_0.Idle + IdleBreak
  ids=id
  nodes [AnimStateNode_0 = Idle Anims() {
    is=title AnimStateNode
    stateEntered {guid=#1}
    stateFullyBlended {guid=#2}
    stateLeft {guid=#3}
   }]
 }
 Idle Anims {
  path=AnimGraph.AnimGraphNode_StateMachine_0.Idle + IdleBreak.AnimStateNode_0.Idle Anims
  nodes [
   n1 = Random Sequence Player!() -> (Pose:Pose Link Structure) {
    is=title AnimGraphNode_RandomPlayer
    entries [
     {
      chanceToPlay=0.75
      maxLoopCount=0
      maxPlayRate=1
      minLoopCount=0
      minPlayRate=1
      sequence=&Backpack_TowerMilk_Idle
      blendIn {blendOption=Linear, blendTime=0.2, customCurve=null}
     }
     {
      chanceToPlay=0.25
      maxLoopCount=0
      maxPlayRate=1
      minLoopCount=0
      minPlayRate=1
      sequence=/BRCosmetics/Accessories/FORT_Backpacks/Backpack_M_MED_TowerMilk/Animations/Backpack_TowerMilk_IdleBreak
      blendIn {blendOption=Linear, blendTime=0.2, customCurve=null}
     }
    ]
   }
   Output Animation Pose(Result=@n1)
  ]
 }
 EventGraph {nodes [TryGetPawnOwner()]}
]
variables [Play_IdleBreak {type=boolean}]
